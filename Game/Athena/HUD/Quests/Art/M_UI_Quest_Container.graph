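# Translucent UI material for the Athena HUD quest container, with static switches for toast, arrow and color-overlay variants.
# The container texture's R channel gives the sharpened nine-slice shape, while G gives a scalable, offset arrow mask.
# A left-side band takes the rarity color, and an aspect-corrected, screen-aligned triangle grid pans with a horizontal falloff.
# Toast adds triangles to color and opacity, while Hover blends toward a triangle-filled opacity mask.
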Material /Game/Athena/HUD/Quests/Art/M_UI_Quest_Container
graph {
 properties {blendMode=BLEND_Translucent, materialDomain=MD_UI, pixelDepthOffsetMode=PDOM_Legacy, usedShadingModels=Default Lit}
 nodes [
  n1 = TextureObject() {texture=&/T_Quest_Container}
  n2 = TextureSample(UVs=MaterialFunctionCall() {materialFunction=&GetUserInterfaceUV}."9-Slice UV (No Tiling)", Tex=@n1) {texture=&T_Quest_Container}
  n3 = ScalarParameter() {parameterName=EdgeOffSet}
  n4 = ScalarParameter() {defaultValue=4, parameterName=EdgeSharpness}
  n5 = Saturate(None=Multiply(A=Subtract(A=@n2.R, B=@n3), B=@n4))
  n6 = MaterialFunctionCall() {materialFunction=&GetUserInterfaceUV}
  n7 = LinearInterpolate(
   A=ScalarParameter() {defaultValue=1, parameterName=TriangleFallOff-Start}
   B=ScalarParameter() {parameterName=TriangleFallOff-End}
   Alpha=Saturate(None=ComponentMask(None=@n6.Normalized UV) {r=true})
  ) {constA=1, constB=0.5}
  n8 = MaterialFunctionCall(Float2=MaterialFunctionCall() {materialFunction=&ScreenResolution}.Visible Resolution) {materialFunction=&BreakOutFloat2Components}
  n9 = Max(A=@n8.G)
  n10 = Multiply(
   A=AppendVector(A=Divide(A=@n8.R, B=@n9), B=Constant() {r=1})
   B=MaterialFunctionCall() {materialFunction=&ScreenAlignedUVs}."X 100%, Y 100%"
  )
  n11 = ScalarParameter() {defaultValue=-0.0571, parameterName=Tri Angle}
  n12 = ScalarParameter() {defaultValue=0.0025, parameterName=PanSpeed}
  n13 = Constant()
  n14 = Multiply(
   A=Panner(Coordinate=Rotator(Coordinate=@n10, Time=@n11) {speed=6.2832}, Speed=AppendVector(A=@n12, B=@n13))
   B=ScalarParameter() {defaultValue=6, parameterName=Tri Tile Amount}
  )
  n15 = Saturate(None=Multiply(
    A=TextureSample(UVs=@n14) {texture=/Game/Global/Textures/Default/Masks_and_Gradients/T-TriangleGrid}.R
    B=ScalarParameter() {defaultValue=1, parameterName=TriangleDensity, sliderMax=1}
   ))
  n16 = Saturate(None=LinearInterpolate(A=@n15, Alpha=@n7) {constB=0})
  n17 = ScalarParameter() {defaultValue=0.5, parameterName=TriangleOverlay-Opacity}
  n18 = Saturate(None=Add(A=Multiply(A=@n16, B=@n17), B=OneMinus(None=@n7)))
  n19 = ComponentMask(None=OneMinus(None=TextureCoordinate())) {r=true}
  n20 = ScalarParameter() {defaultValue=0.75, parameterName=RarityBand-EdgeOffSet}
  n21 = Saturate(None=Multiply(A=Subtract(A=@n19, B=@n20), B=ScalarParameter() {defaultValue=128, parameterName=RarityBand-EdgeSharpness}))
  n22 = Saturate(None=Subtract(A=@n18, B=@n21))
  n23 = LinearInterpolate(A=ScalarParameter() {defaultValue=0.5, parameterName=Container-Opacity}, Alpha=@n21)
  n24 = Multiply(A=StaticSwitchParameter(True=Add(A=@n22, B=@n23), False=@n23) {parameterName=bIsToast}, B=@n5)
  n25 = ScalarParameter() {defaultValue=1, parameterName=TriangleFallOff-Fill}
  n26 = Multiply(A=Saturate(None=LinearInterpolate(A=@n15, B=@n25, Alpha=@n7)), B=ScalarParameter() {defaultValue=0.5, parameterName=Container-Opacity})
  n27 = Add(A=DotProduct(A=@n5, B=@n21), B=Saturate(None=DotProduct(A=@n26, B=@n5)))
  n28 = ScalarParameter() {parameterName=Hover}
  n29 = Add(
   A=Multiply(
    A=MaterialFunctionCall() {materialFunction=&GetUserInterfaceUV}.Normalized UV
    B=ScalarParameter() {parameterName=Scale}
   )
   B=AppendVector(A=ScalarParameter() {parameterName=u-Offset}, B=ScalarParameter() {parameterName=y-Offset})
  )
  n30 = Multiply(A=Subtract(A=TextureSample(UVs=@n29, Tex=@n1) {texture=&T_Quest_Container}.G, B=@n3), B=@n4)
  Multiply(
   A=StaticSwitchParameter(True=Saturate(None=@n30), False=LinearInterpolate(A=@n24, B=@n27, Alpha=@n28)) {parameterName=bIsArrow}
   B=ScalarParameter() {parameterName=Opacity}
  )
  n31 = VectorParameter() {parameterName=Triangle_color, defaultValue=<rgba 0 0.004 0.0723 1>}
  n32 = VectorParameter() {parameterName=Background_color, defaultValue=<rgba 0 0.004 0.0723 1>}
  n33 = VectorParameter() {parameterName=RGB-Rarity1, defaultValue=<rgba 0 0.004 0.0723 1>}
  n34 = StaticSwitchParameter(
   True=VectorParameter() {parameterName=Overlay_Color, defaultValue=<rgba 1 1 1 1>}.RGB
   False=StaticSwitchParameter(True=@n33.RGB, False=LinearInterpolate(A=@n32.RGB, B=@n33.RGB, Alpha=@n21)) {parameterName=bIsArrow}
  ) {parameterName=IsOverlay}
  StaticSwitchParameter(True=Add(A=LinearInterpolate(B=@n31.RGB, Alpha=@n18), B=@n34), False=@n34) {parameterName=bIsToast}
 ]
}
